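; Gameplay ability for a player picking up and carrying another player, with NPC-folder placement and carry-distance quest tracking.
; Activation holsters the weapon, unless block-holster tags apply, then plays the pickup montage.
; A vehicle-hit binding cancels only during pickup; completion unbinds it and records the location.
; Ending unholsters, stops carrying and, server-side after completed pickup, reports straight-line distance as a quest stat; the long-distance feat branch is empty.
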(event Custom|OnFortPlayerHitByVehicle_Event
  (params (InstigatingVehicle :type Fort%20Athena%20Vehicle%20Object%20Reference))
  ;[1 1 1] Interrupting carry pickup. Can't use cancel tags because we only want to interrupt while picking up, not after.
  (Ability|CancelAbility))

(event Ability|EventActivateAbility
  (bind _playerpawn (Variables|Default|GetPlayerPawn))
  (|CallParentFunction)
  (bind (_event_data _fort_player_pawn) (Setup|SetupGenericAbility false false)
    (:Output
      (bind _output_get (Variables|Default|SetPlayerPawn _fort_player_pawn))
      ;[0 0.07755 0.9062 0.5] Allow items with tags to block holster. This is to allow  the item visual to persist ( character parts ) and potentially to help communicate that certain actions from the item may still function
      (if (not (GameplayTags|HasAnyMatchingGameplayTags _output_get (Variables|Default|GetBlockHolsterTags)))
        (Weapon|SetHolsterWeapon _output_get 0 true false (Variables|Default|GetDBNOCarryHolster)))
      (Ability|Tasks|PlayMontageAndWait "None" (Utilities|Struct|BreakFortGameplayAbilityMontageInfo (Variables|Animation|GetMontageInfo))
        (:then
          ;[1 1 1] Interrupting carry pickup. Can't use cancel tags because we only want to interrupt while picking up, not after.
          (bind _asfort_player_pawn_athena (Utilities|Casting|CastToFortPlayerPawnAthena _playerpawn))
          (EventDispatchers|AssignOnFortPlayerHitbyVehicle _asfort_player_pawn_athena (delegate Custom|OnFortPlayerHitByVehicle_Event)))
        (:OnCompleted
          (Variables|AthenaQuest|SetbCompletedPickup true)
          (Variables|AthenaQuest|SetPickupLocation (Transformation|GetWorldLocation (Variables|/Script/Engine.Character|GetCapsuleComponent (Variables|Default|GetPlayerPawn))))
          (bind _asfort_player_pawn_athena_2 (Utilities|Casting|CastToFortPlayerPawnAthena _playerpawn))
          (EventDispatchers|UnbindEventfromOnFortPlayerHitbyVehicle _asfort_player_pawn_athena_2 (delegate Custom|OnFortPlayerHitByVehicle_Event)))))))

(event Ability|EventOnEndAbility
  (params (bWasCancelled :type Boolean))
  (bind _pickuplocation (Variables|AthenaQuest|GetPickupLocation))
  (bind _playerpawn (Variables|Default|GetPlayerPawn))
  (bind _capsulecomponent (Variables|/Script/Engine.Character|GetCapsuleComponent _playerpawn))
  (bind _returnvalue (Transformation|GetWorldLocation _capsulecomponent))
  (bind _returnvalue_2 (Math|Vector|Distance(Vector) _returnvalue _pickuplocation))
  (bind _returnvalue_3 (Math|Float|Round _returnvalue_2))
  (bind _playerpawn_2 (Variables|Default|GetPlayerPawn))
  (bind _returnvalue_4 (Fortnite|GetFortPlayerControllerfromActor _playerpawn_2))
  (bind _returnvalue_5 (Quests|GetQuestManager _returnvalue_4 "Invalid"))
  (bind _returnvalue_6 (/ _returnvalue_3 100))
  (bind _playerpawn_3 (Variables|Default|GetPlayerPawn))
  (|CallParentFunction bWasCancelled)
  (Utilities|IsValid _playerpawn_3
    (:Is%20Valid
      ;[0 0.07755 0.9062 0.5] Allow items with tags to block holster. This is to allow  the item visual to persist ( character parts ) and potentially to help communicate that certain actions from the item may still function
      (if (not (GameplayTags|HasAnyMatchingGameplayTags _playerpawn_3 (Variables|Default|GetBlockHolsterTags)))
        (Weapon|SetHolsterWeapon _playerpawn_3 0 false false (Variables|Default|GetDBNOCarryHolster)))
      (Class|FortPlayerPawn|StopCarryingPlayer _playerpawn_3 true)
      ;[0.006669 0.739 1 0.107] Track distance traveled from start of pickup and send a quest event
      (if (Networking|IsServer)
        (Utilities|FlowControl|Sequence
          (:then_0
            (if (Variables|AthenaQuest|GetbCompletedPickup)
              (if (Networking|IsServer)
                (Utilities|IsValid _playerpawn_2
                  (:Is%20Valid
                    (Utilities|IsValid _returnvalue_5
                      (:Is%20Valid
                        (Quest|SendComplexCustomStatEvent _returnvalue_5 0 "(GameplayTags=((TagName=\"Athena.Quests.CarryPlayerDistance\")))" "(GameplayTags=)" _returnvalue_6)
                        (bind (_questisvalid _questcompleted) (Default|AthenaCheckQuestandObjectiveCompletion (Fortnite|GetFortPlayerControllerfromActor (Variables|Default|GetPlayerPawn)) 0 "Feat_Season11_CarryTeammateDistance"))
                        (if (and _questisvalid (not _questcompleted))
                          (if (>= _returnvalue_6 (Variables|AthenaQuest|GetCarryLongDistanceThreshold))
                            _)))))))))
          (:then_1
            (Variables|AthenaQuest|SetbCompletedPickup false)))))))
