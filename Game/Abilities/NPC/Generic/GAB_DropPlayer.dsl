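; Gameplay ability in the generic NPC set that makes a Fortnite player pawn stop carrying another player.
; On activation it runs the generic setup, stores the player pawn and plays a montage through a wait-for-trigger task, followed by EndAbility.
; StopCarryingPlayer is called on the montage trigger and again in OnEndAbility, so the drop also happens when the ability ends without that trigger.
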(event Ability|EventActivateAbility
  (|CallParentFunction)
  (bind (_event_data _fort_player_pawn) (Setup|SetupGenericAbility false false)
    (:Output
      (Variables|Default|SetPlayerPawn _fort_player_pawn)
      (Ability|Tasks|PlayMontageInfoAndWaitForTrigger "None" (Variables|Animation|GetMontageInfo) "SimulateOnServer"
        (:Triggered
          (Class|FortPlayerPawn|StopCarryingPlayer (Variables|Default|GetPlayerPawn) false))
        (:Cancelled
          _)
        (:Completed
          _))
      (Ability|EndAbility))))

(event Ability|EventOnEndAbility
  (params (bWasCancelled :type Boolean))
  (|CallParentFunction bWasCancelled)
  (Class|FortPlayerPawn|StopCarryingPlayer (Variables|Default|GetPlayerPawn) false))
