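; Generic NPC ability that throws the player pawn carried by the NPC.
; On activation it runs generic setup, stores the player pawn and plays a montage through a trigger-waiting task with server simulation.
; The montage trigger calls LocalThrowCarriedPlayer on that pawn.
; Ability end repeats the throw call, including on cancellation.
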(event Ability|EventActivateAbility
  (|CallParentFunction)
  (bind (_event_data _fort_player_pawn) (Setup|SetupGenericAbility false false)
    (:Output
      (Variables|Default|SetPlayerPawn _fort_player_pawn)
      (Ability|Tasks|PlayMontageInfoAndWaitForTrigger "None" (Variables|Animation|GetMontageInfo) "SimulateOnServer"
        (:Triggered
          (Class|FortPlayerPawn|LocalThrowCarriedPlayer (Variables|Default|GetPlayerPawn)))
        (:Cancelled
          _)
        (:Completed
          _))
      (Ability|EndAbility))))

(event Ability|EventOnEndAbility
  (params (bWasCancelled :type Boolean))
  (|CallParentFunction bWasCancelled)
  (Class|FortPlayerPawn|LocalThrowCarriedPlayer (Variables|Default|GetPlayerPawn)))
